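# UI material for a player survey radio option circle, drawn procedurally with translucent blending.
# One MF_UI_Circle call supplies the outer outline ring and a second, half-size call supplies the filled center dot.
# The OptionSelected static switch, on by default, selects ring plus dot or the ring alone, so selected state is a compile-time permutation rather than a runtime parameter.
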Material /FortniteCoreUI/UI/PlayerSurvey/Materials/M_PlayerSurvey_RadioOptionCircle
graph {
 properties {
  bAutomaticallySetUsageInEditor=false
  bUsedWithStaticMesh=false
  blendMode=BLEND_Translucent
  materialDomain=MD_UI
  usedShadingModels=Default Lit
 }
 nodes [
  n1 = Constant() {r=0.5}
  n2 = MaterialFunctionCall(Size=Constant() {r=1}, OutlineThickness=Constant() {r=0.15}, OutlineAlignment=Constant() {r=-1}) {materialFunction=&MF_UI_Circle}
  StaticSwitchParameter(
   True=Add(A=MaterialFunctionCall(Size=@n1) {materialFunction=&MF_UI_Circle}.Fill, B=@n2.Outline) {constB=0}
   False=@n2.Outline
  ) {defaultValue=true, parameterName=OptionSelected}
 ]
}
